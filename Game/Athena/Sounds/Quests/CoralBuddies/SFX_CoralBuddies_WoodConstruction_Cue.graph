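# Sound cue for the wood construction effect in the CoralBuddies quest sounds.
# A single wave player feeds the output directly, with no randomization or modulation.
# It routes through the in-game SFX sound class and enables override attenuation with its own falloff distance, so the referenced chest attenuation asset is bypassed.
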SoundCue /Game/Athena/Sounds/Quests/CoralBuddies/SFX_CoralBuddies_WoodConstruction_Cue
root=WavePlayer_1
nodes [{id=WavePlayer_1, properties {soundWaveAssetPtr=&/SFX_CoralBuddies_WoodConstruction}}]
properties {
 attenuationSettings=/Game/Sounds/Attenuation/Fort_Chests_Att
 bOverrideAttenuation=true
 maxDistance=2400
 soundClassObject=&Fort_SFX_InGame
 volumeMultiplier=1
 attenuationOverrides {falloffDistance=2000}
}
